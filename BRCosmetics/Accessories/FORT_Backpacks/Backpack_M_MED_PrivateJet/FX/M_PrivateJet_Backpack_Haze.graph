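# Additive particle material for the haze effect on the Private Jet backpack cosmetic.
# Adds scene color to particle color, scales it, then runs it through time-of-day emissive boost and cosmetic eye-adaptation functions.
# A powered radial sphere mask times particle alpha forms a procedural soft disc, depth-faded on the default shading path.
# The mobile path skips scene color, both functions and depth fade, using scaled particle color and the plain mask.
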Material /BRCosmetics/Accessories/FORT_Backpacks/Backpack_M_MED_PrivateJet/FX/M_PrivateJet_Backpack_Haze
graph {
 properties {
  bUsedWithMeshDeformer=true
  bUsedWithNiagaraSprites=true
  bUsedWithSkeletalMesh=true
  blendMode=BLEND_Additive
  pixelDepthOffsetMode=PDOM_Legacy
  usedShadingModels=Default Lit
 }
 nodes [
  n1 = ParticleColor()
  n2 = Add(A=Multiply(A=SceneColor!(UVs).RGB, B=ScalarParameter() {defaultValue=1, parameterName=SceneColorIntensity}), B=@n1.RGB)
  n3 = ScalarParameter() {defaultValue=0.1, parameterName=EmissiveIntensity}
  n4 = MaterialFunctionCall!(
   Emissive=MaterialFunctionCall!(In=Multiply(A=@n2, B=@n3)) {materialFunction=/Game/Weapons/WeaponSkins/MaterialFunction/BoostEmissivebyTOD}.Result
   Alpha=ScalarParameter() {defaultValue=1, group=Cosmetic Emissive EyeAdaptation, parameterName=Cosmetic Emissive EyeAdaptation Alpha}
   Material
  ) {materialFunction=/Game/Packages/Fortress_SharedMaterials/Base_Material_Functions/BaseCharacterShaderFunctions/MF_CosmeticEmissiveEyeAdaptation}
  ShadingPathSwitch(Default=@n4.Emissive, Mobile=Multiply(A=ScalarParameter() {defaultValue=0.15, parameterName=EmissiveIntensity_Mobile}, B=@n1.RGB))
  n5 = Power(
   Base=SphereMask(
    A=TextureCoordinate()
    B=Constant() {r=0.5}
    Radius=ScalarParameter() {defaultValue=0.5, parameterName=Radius}
    Hardness=ScalarParameter() {parameterName=Hardness}
   )
   Exp=ScalarParameter() {defaultValue=2, parameterName=Power}
  )
  n6 = Multiply(A=@n1.A, B=NamedRerouteDeclaration(None=Saturate(None=@n5)) {name=SphereMask, nodeColor=<rgba 0 1 0.0625 1>})
  n7 = ScalarParameter() {defaultValue=50, parameterName=FadeDistance}
  ShadingPathSwitch(Default=Multiply(A=@n6, B=DepthFade(FadeDistance=@n7)), Mobile=@n6)
 ]
}
