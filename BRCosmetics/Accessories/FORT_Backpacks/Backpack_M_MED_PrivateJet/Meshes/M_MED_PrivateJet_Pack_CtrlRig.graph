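# Control Rig for the PrivateJet backpack that procedurally adjusts its cloth bones and center-of-gravity bone (JNT_Cog).
# Distances from cloth tip bones to the Back_Loc and Back_Loc_2 controls become clamped inverse weights, so adjustments strengthen with proximity.
# With SkyDiveCtrl off, four weighted passes rotate back and side cloth bones and nudge JNT_Cog.
# With it on, JNT_Cog gets a fixed offset, followed by one weighted back-cloth rotation.
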ControlRigBlueprint /BRCosmetics/Accessories/FORT_Backpacks/Backpack_M_MED_PrivateJet/Meshes/M_MED_PrivateJet_Pack_CtrlRig
graphs [
 RigVMFunctionLibrary {links [], nodes []}
 RigVMModel {nodes [
   RigUnit_BeginExecution = &RigUnit_BeginExecution()
   RigVMFunction_ControlFlowBranch = &RigVMFunction_ControlFlowBranch(ExecuteContext=@RigUnit_BeginExecution.ExecutePin, Condition=RigVMVariableNode() {variable=SkyDiveCtrl}.Value)
   n1 = &RigUnit_GetTransform(Item=(Type=Control,Name="Back_Loc"))
   n2 = &RigUnit_GetTransform(Item=(Type=Bone,Name="JNT_cloth_05_B"))
   RigVMFunction_MathQuaternionFromEuler = &RigVMFunction_MathQuaternionFromEuler(Euler=(X=0,Y=40,Z=0))
   &RigUnit_ModifyTransforms(
    ItemToModify=((Item=(Type=Bone,Name="JNT_cloth_01_B"),Transform=(Rotation=(X=0,Y=0,Z=0,W=1),Translation=(X=0,Y=0,Z=0),Scale3D=(X=1,Y=1,Z=1))),(Item=(Type=Bone,Name="JNT_Cog"),Transform=(Rotation=(X=0,Y=0,Z=0,W=1),Translation=(X=0,Y=0,Z=-1),Scale3D=(X=1,Y=1,Z=1))))
    WeightMinimum=0
    Weight=&RigVMFunction_MathFloatRemap(
     SourceMaximum=12
     SourceMinimum=0
     TargetMaximum=0
     TargetMinimum=1
     bClamp=true
     Value=&RigVMFunction_MathVectorDistance(B=@n1.Transform.Translation, A=@n2.Transform.Translation).Result
    ).Result
    ExecutePin=@RigVMFunction_ControlFlowBranch.False
    ItemToModify.0.Transform.Rotation=@RigVMFunction_MathQuaternionFromEuler.Result
   )
   n3 = &RigUnit_GetTransform(Item=(Type=Control,Name="Back_Loc_2"))
   n4 = &RigUnit_GetTransform(Item=(Type=Bone,Name="JNT_cloth_05_B"))
   > &RigUnit_ModifyTransforms(
    ItemToModify=((Item=(Type=Bone,Name="JNT_cloth_01_B"),Transform=(Rotation=(X=0,Y=0,Z=0,W=1),Translation=(X=0,Y=0,Z=0),Scale3D=(X=1,Y=1,Z=1))),(Item=(Type=Bone,Name="JNT_Cog"),Transform=(Rotation=(X=0,Y=0,Z=0,W=1),Translation=(X=0,Y=0,Z=-1),Scale3D=(X=1,Y=1,Z=1))))
    WeightMinimum=0
    Weight=&RigVMFunction_MathFloatRemap(
     SourceMaximum=10
     SourceMinimum=0
     TargetMaximum=0
     TargetMinimum=1
     bClamp=true
     Value=&RigVMFunction_MathVectorDistance(B=@n3.Transform.Translation, A=@n4.Transform.Translation).Result
    ).Result
    ItemToModify.0.Transform.Rotation=&RigVMFunction_MathQuaternionFromEuler(Euler=(X=0,Y=40,Z=0)).Result
   )
   n5 = &RigUnit_GetTransform(Item=(Type=Control,Name="Back_Loc"))
   n6 = &RigUnit_GetTransform(Item=(Type=Bone,Name="JNT_cloth_06_05_L"))
   > &RigUnit_ModifyTransforms(
    ItemToModify=((Item=(Type=Bone,Name="JNT_cloth_06_05_L"),Transform=(Rotation=(X=0,Y=0,Z=0,W=1),Translation=(X=0,Y=0,Z=0),Scale3D=(X=1,Y=1,Z=1))),(Item=(Type=Bone,Name="JNT_Cog"),Transform=(Rotation=(X=0,Y=0,Z=0,W=1),Translation=(X=0,Y=0,Z=-1),Scale3D=(X=1,Y=1,Z=1))))
    WeightMinimum=0
    Weight=&RigVMFunction_MathFloatRemap(
     SourceMaximum=10
     SourceMinimum=0
     TargetMaximum=0
     TargetMinimum=1
     bClamp=true
     Value=&RigVMFunction_MathVectorDistance(B=@n5.Transform.Translation, A=@n6.Transform.Translation).Result
    ).Result
    ItemToModify.0.Transform.Rotation=&RigVMFunction_MathQuaternionFromEuler(Euler=(X=0,Y=45,Z=0)).Result
   )
   n7 = &RigUnit_GetTransform(Item=(Type=Control,Name="Back_Loc"))
   n8 = &RigUnit_GetTransform(Item=(Type=Bone,Name="JNT_cloth_06_05_R"))
   > &RigUnit_ModifyTransforms(
    ItemToModify=((Item=(Type=Bone,Name="JNT_cloth_06_05_R"),Transform=(Rotation=(X=0,Y=0,Z=0,W=1),Translation=(X=0,Y=0,Z=0),Scale3D=(X=1,Y=1,Z=1))),(Item=(Type=Bone,Name="JNT_Cog"),Transform=(Rotation=(X=0,Y=0,Z=0,W=1),Translation=(X=0,Y=0,Z=-1),Scale3D=(X=1,Y=1,Z=1))))
    WeightMinimum=0
    Weight=&RigVMFunction_MathFloatRemap(
     SourceMaximum=10
     SourceMinimum=0
     TargetMaximum=0
     TargetMinimum=1
     bClamp=true
     Value=&RigVMFunction_MathVectorDistance(B=@n7.Transform.Translation, A=@n8.Transform.Translation).Result
    ).Result
    ItemToModify.0.Transform.Rotation=&RigVMFunction_MathQuaternionFromEuler(Euler=(X=0,Y=45,Z=0)).Result
   )
   &RigUnit_ModifyTransforms(
    ItemToModify=((Item=(Type=Bone,Name="JNT_Cog"),Transform=(Rotation=(X=0,Y=0,Z=0,W=1),Translation=(X=3,Y=0,Z=-5),Scale3D=(X=1,Y=1,Z=1))))
    Weight=1
    WeightMinimum=0
    ExecutePin=@RigVMFunction_ControlFlowBranch.True
   )
   n9 = &RigUnit_GetTransform(Item=(Type=Control,Name="Back_Loc_2"))
   n10 = &RigUnit_GetTransform(Item=(Type=Bone,Name="JNT_cloth_05_B"))
   > &RigUnit_ModifyTransforms(
    ItemToModify=((Item=(Type=Bone,Name="JNT_cloth_01_B"),Transform=(Rotation=(X=0,Y=0,Z=0,W=1),Translation=(X=0,Y=0,Z=0),Scale3D=(X=1,Y=1,Z=1))))
    WeightMinimum=0
    Weight=&RigVMFunction_MathFloatRemap(
     SourceMaximum=10
     SourceMinimum=0
     TargetMaximum=0
     TargetMinimum=1
     bClamp=true
     Value=&RigVMFunction_MathVectorDistance(B=@n9.Transform.Translation, A=@n10.Transform.Translation).Result
    ).Result
    ItemToModify.0.Transform.Rotation=&RigVMFunction_MathQuaternionFromEuler(Euler=(X=0,Y=40,Z=0)).Result
   )
   RigVMFunction_Sequence = &RigVMFunction_Sequence(ExecuteContext=())
  ]}
]
hierarchy [
 root {loc=<0 0 0>, rot=<0 0 0 1>}
 pack {parent=root, loc=<0 -0.732 0>, rot=<0 0 0 1>}
 JNT_Cog {parent=pack, loc=<0.0385 -5.9107 -0.3975>, rot=<0.4176 0.4199 0.5682 -0.5713>}
 JNT_neck {parent=JNT_Cog, loc=<0.496 0 0>, rot=<0 0 0 1>}
 JNT_head {parent=JNT_neck, loc=<1.8121 0 0>, rot=<0.0118 -0.1288 0.0054 -0.9916>}
 JNT_upperhead {parent=JNT_head, loc=<3.2523 -0.0311 1.2839>, rot=<0 -0.7071 0 -0.7071>}
 JNT_ear_01_L {parent=JNT_upperhead, loc=<-0.5887 -1.8291 2.4659>, rot=<0.0373 -0.8844 -0.0098 0.4651>}
 JNT_ear_01_ExL {parent=JNT_ear_01_L, loc=<2.4623 -0.1686 0.0945>, rot=<-0.0008 -0.0209 -0.0372 -0.9991>}
 JNT_ear_01_ExL_R {parent=JNT_ear_01_ExL, loc=<0 1.3244 0.0592>, rot=<0.5 0.5 0.5 0.5>}
 JNT_ear_01_ExL_L {parent=JNT_ear_01_ExL, loc=<0 -1.3351 0.0588>, rot=<0.5 -0.5 0.5 -0.5>}
 JNT_ear_02_L {parent=JNT_ear_01_ExL, loc=<2.2663 0 0>, rot=<0 0 0 1>}
 JNT_ear_02_L_R {parent=JNT_ear_02_L, loc=<0 1.6746 0.0592>, rot=<0.5 0.5 0.5 0.5>}
 JNT_ear_02_L_L {parent=JNT_ear_02_L, loc=<0 -1.7029 0.0588>, rot=<0.5 -0.5 0.5 -0.5>}
 JNT_ear_02_ExL {parent=JNT_ear_02_L, loc=<1.7307 0 0>, rot=<0 0 0 1>}
 JNT_ear_02_ExL_R {parent=JNT_ear_02_ExL, loc=<0 1.6746 0.0592>, rot=<0.5 0.5 0.5 0.5>}
 JNT_ear_02_ExL_L {parent=JNT_ear_02_ExL, loc=<0 -1.7029 0.0588>, rot=<0.5 -0.5 0.5 -0.5>}
 JNT_ear_03_L {parent=JNT_ear_02_ExL, loc=<1.9659 0 0>, rot=<0 0 0 1>}
 JNT_ear_03_L_R {parent=JNT_ear_03_L, loc=<0 1.6009 0.0626>, rot=<0.5 0.5 0.5 0.5>}
 JNT_ear_03_L_L {parent=JNT_ear_03_L, loc=<0 -1.6242 0.0626>, rot=<0.5 -0.5 0.5 -0.5>}
 JNT_ear_03_ExL {parent=JNT_ear_03_L, loc=<2.0654 0 0>, rot=<0 0 0 1>}
 JNT_ear_03_ExL_R {parent=JNT_ear_03_ExL, loc=<0 1.4507 0.0592>, rot=<0.5 0.5 0.5 0.5>}
 JNT_ear_03_ExL_L {parent=JNT_ear_03_ExL, loc=<0 -1.4891 0.0588>, rot=<0.5 -0.5 0.5 -0.5>}
 JNT_ear_04_L {parent=JNT_ear_03_ExL, loc=<2.0947 0 0>, rot=<0 0 0 1>}
 JNT_ear_05_L {parent=JNT_ear_04_L, loc=<2.4479 0 0>, rot=<0 0 0 1>}
 JNT_ear_05_L_R {parent=JNT_ear_05_L, loc=<-0.0002 0.9645 0.0626>, rot=<0.5 0.5 0.5 0.5>}
 JNT_ear_05_L_L {parent=JNT_ear_05_L, loc=<-0.0002 -0.993 0>, rot=<0.5 -0.5 0.5 -0.5>}
 JNT_ear_05_ExL {parent=JNT_ear_05_L, loc=<1.1806 0 0>, rot=<0 0 0 1>}
 JNT_ear_05_ExL_R {parent=JNT_ear_05_ExL, loc=<0 0.8534 0.0592>, rot=<0.5 0.5 0.5 0.5>}
 JNT_ear_05_ExL_L {parent=JNT_ear_05_ExL, loc=<0 -0.8099 0.0588>, rot=<0.5 -0.5 0.5 -0.5>}
 JNT_ear_06_L {parent=JNT_ear_05_ExL, loc=<1.3586 0 0>, rot=<0 0 0 1>}
 JNT_ear_07_L {parent=JNT_ear_06_L, loc=<1.9302 0 0>, rot=<0 0 0 1>}
 JNT_ear_08_L {parent=JNT_ear_07_L, loc=<1.9303 0 0>, rot=<0 0 0 1>}
 JNT_ear_08_L_R {parent=JNT_ear_08_L, loc=<0 0.3848 0.0625>, rot=<0.5 0.5 0.5 0.5>}
 JNT_ear_08_L_L {parent=JNT_ear_08_L, loc=<0 -0.3978 0.0628>, rot=<0.5 -0.5 0.5 -0.5>}
 JNT_ear_07_L_R {parent=JNT_ear_07_L, loc=<0 0.5665 0.0626>, rot=<0.5 0.5 0.5 0.5>}
 JNT_ear_07_L_L {parent=JNT_ear_07_L, loc=<0 -0.5871 0.0626>, rot=<0.5 -0.5 0.5 -0.5>}
 JNT_ear_06_L_R {parent=JNT_ear_06_L, loc=<-0.0001 0.7521 0.0626>, rot=<0.5 0.5 0.5 0.5>}
 JNT_ear_06_L_L {parent=JNT_ear_06_L, loc=<-0.0001 -0.7777 0.0626>, rot=<0.5 -0.5 0.5 -0.5>}
 JNT_ear_04_L_R {parent=JNT_ear_04_L, loc=<0 1.2881 0.0626>, rot=<0.5 0.5 0.5 0.5>}
 JNT_ear_04_L_L {parent=JNT_ear_04_L, loc=<0 -1.3165 0.0626>, rot=<0.5 -0.5 0.5 -0.5>}
 JNT_ear_01_R {parent=JNT_upperhead, loc=<-0.6529 1.861 2.4255>, rot=<-0.027 -0.8848 0.0276 0.4644>}
 JNT_ear_01_ExR {parent=JNT_ear_01_R, loc=<2.4636 0.1861 0.0944>, rot=<0.0008 -0.0209 0.0372 -0.9991>}
 JNT_ear_01_ExR_R {parent=JNT_ear_01_ExR, loc=<0 1.3244 0.0592>, rot=<0.5 0.5 0.5 0.5>}
 JNT_ear_01_ExR_L {parent=JNT_ear_01_ExR, loc=<0 -1.3351 0.0588>, rot=<0.5 -0.5 0.5 -0.5>}
 JNT_ear_02_R {parent=JNT_ear_01_ExR, loc=<2.2663 -0.0133 0>, rot=<0 0 0 1>}
 JNT_ear_02_R_R {parent=JNT_ear_02_R, loc=<-0.0001 1.7029 0.0588>, rot=<-0.5 -0.5 -0.5 -0.5>}
 JNT_ear_02_R_L {parent=JNT_ear_02_R, loc=<-0.0001 -1.6746 0.0592>, rot=<0.5 -0.5 0.5 -0.5>}
 JNT_ear_02_ExR {parent=JNT_ear_02_R, loc=<1.7307 0.0133 0>, rot=<0 0 0 1>}
 JNT_ear_02_ExR_R {parent=JNT_ear_02_ExR, loc=<0 1.6746 0.0592>, rot=<0.5 0.5 0.5 0.5>}
 JNT_ear_02_ExR_L {parent=JNT_ear_02_ExR, loc=<0 -1.7029 0.0588>, rot=<0.5 -0.5 0.5 -0.5>}
 JNT_ear_03_R {parent=JNT_ear_02_ExR, loc=<1.9926 -0.0176 0>, rot=<0 0 0 1>}
 JNT_ear_03_R_L {parent=JNT_ear_03_R, loc=<0 -1.6009 0.0626>, rot=<0.5 -0.5 0.5 -0.5>}
 JNT_ear_03_R_R {parent=JNT_ear_03_R, loc=<0 1.6242 0.0626>, rot=<-0.5 -0.5 -0.5 -0.5>}
 JNT_ear_03_ExR {parent=JNT_ear_03_R, loc=<2.0654 0.0133 0>, rot=<0 0 0 1>}
 JNT_ear_03_ExR_R {parent=JNT_ear_03_ExR, loc=<0 1.4507 0.0592>, rot=<0.5 0.5 0.5 0.5>}
 JNT_ear_03_ExR_L {parent=JNT_ear_03_ExR, loc=<0 -1.4891 0.0588>, rot=<0.5 -0.5 0.5 -0.5>}
 JNT_ear_04_R {parent=JNT_ear_03_ExR, loc=<2.0851 -0.0133 0>, rot=<0 0 0 1>}
 JNT_ear_05_R {parent=JNT_ear_04_R, loc=<2.5738 0 0>, rot=<0 0 0 1>}
 JNT_ear_05_R_L {parent=JNT_ear_05_R, loc=<-0.0002 -0.9645 0.0626>, rot=<0.5 -0.5 0.5 -0.5>}
 JNT_ear_05_R_R {parent=JNT_ear_05_R, loc=<-0.0002 0.993 0>, rot=<-0.5 -0.5 -0.5 -0.5>}
 JNT_ear_05_ExR {parent=JNT_ear_05_R, loc=<1.1806 0.0133 0>, rot=<0 0 0 1>}
 JNT_ear_05_ExR_R {parent=JNT_ear_05_ExR, loc=<0 0.8534 0.0592>, rot=<0.5 0.5 0.5 0.5>}
 JNT_ear_05_ExR_L {parent=JNT_ear_05_ExR, loc=<0 -0.8099 0.0588>, rot=<0.5 -0.5 0.5 -0.5>}
 JNT_ear_06_R {parent=JNT_ear_05_ExR, loc=<1.2592 -0.0133 0>, rot=<0 0 0 1>}
 JNT_ear_06_R_L {parent=JNT_ear_06_R, loc=<0 -0.7521 0.0626>, rot=<0.5 -0.5 0.5 -0.5>}
 JNT_ear_06_R_R {parent=JNT_ear_06_R, loc=<0 0.7777 0.0626>, rot=<-0.5 -0.5 -0.5 -0.5>}
 JNT_ear_07_R {parent=JNT_ear_06_R, loc=<1.9302 0 0>, rot=<0 0 0 1>}
 JNT_ear_07_R_L {parent=JNT_ear_07_R, loc=<0 -0.5665 0.0626>, rot=<0.5 -0.5 0.5 -0.5>}
 JNT_ear_07_R_R {parent=JNT_ear_07_R, loc=<0 0.5871 0.0626>, rot=<-0.5 -0.5 -0.5 -0.5>}
 JNT_ear_08_R {parent=JNT_ear_07_R, loc=<1.9303 0 0>, rot=<0 0 0 1>}
 JNT_ear_08_R_L {parent=JNT_ear_08_R, loc=<0 -0.3848 0.0625>, rot=<0.5 -0.5 0.5 -0.5>}
 JNT_ear_08_R_R {parent=JNT_ear_08_R, loc=<0 0.3978 0.0628>, rot=<-0.5 -0.5 -0.5 -0.5>}
 JNT_ear_04_R_L {parent=JNT_ear_04_R, loc=<0 -1.2881 0.0626>, rot=<0.5 -0.5 0.5 -0.5>}
 JNT_ear_04_R_R {parent=JNT_ear_04_R, loc=<0 1.3165 0.0626>, rot=<-0.5 -0.5 -0.5 -0.5>}
 JNT_pumpkin {parent=JNT_upperhead, loc=<12.601 0.2556 2.8527>, rot=<0.0028 0.2771 -0.0099 -0.9608>}
 JNT_lowerlid_L {parent=JNT_upperhead, loc=<3.3536 -1.4217 0.4782>, rot=<-0.0447 0.2734 0.1551 -0.9482>}
 JNT_upperlid_L {parent=JNT_upperhead, loc=<2.7889 -1.3695 1.393>, rot=<-0.0447 0.2734 0.1551 -0.9482>}
 JNT_upperlid_R {parent=JNT_upperhead, loc=<2.739 1.4958 1.3617>, rot=<0.1746 -0.9449 0.0503 -0.2725>}
 JNT_lowerlid_R {parent=JNT_upperhead, loc=<3.3019 1.5476 0.4457>, rot=<0.1746 -0.9449 0.0503 -0.2725>}
 JNT_lowerjaw {parent=JNT_head, loc=<2.6267 0 0.2251>, rot=<0 -0.7071 0 -0.7071>}
 JNT_cloth_01_B {parent=JNT_Cog, loc=<-0.0691 -0.0341 0.8467>, rot=<0.5371 0.0017 0.8435 0.0021>}
 JNT_cloth_02_B {parent=JNT_cloth_01_B, loc=<5.1869 0 -0.1297>, rot=<0 0.0667 0 -0.9978>}
 JNT_cloth_03_B {parent=JNT_cloth_02_B, loc=<6.0895 0 0>, rot=<0 0 0 1>}
 JNT_cloth_04_B {parent=JNT_cloth_03_B, loc=<6.2411 0 0>, rot=<0 0 0 1>}
 JNT_cloth_05_B {parent=JNT_cloth_04_B, loc=<4.6808 0 0>, rot=<0 0 0 1>}
 JNT_cloth_01_01_L {parent=JNT_Cog, loc=<0.0456 -0.2814 -0.9877>, rot=<0.0616 -0.6223 0.0918 -0.7749>}
 JNT_cloth_01_02_L {parent=JNT_cloth_01_01_L, loc=<3.4331 0.1242 -0.2198>, rot=<-0.0012 0.0429 0.0286 -0.9987>}
 JNT_cloth_01_03_L {parent=JNT_cloth_01_02_L, loc=<3.9753 0 0>, rot=<0 -0.0004 0.1039 -0.9946>}
 JNT_cloth_01_04_L {parent=JNT_cloth_01_03_L, loc=<3.9447 0 0>, rot=<-0.0017 0 0 -1>}
 JNT_cloth_02_01_L {parent=JNT_Cog, loc=<0.0337 -0.7394 -0.7521>, rot=<0.1747 -0.5862 0.2532 -0.7495>}
 JNT_cloth_02_02_L {parent=JNT_cloth_02_01_L, loc=<3.689 0 -0.22>, rot=<-0.0008 0.0299 0.028 -0.9992>}
 JNT_cloth_02_03_L {parent=JNT_cloth_02_02_L, loc=<3.2707 0 0>, rot=<0 0 0.0281 -0.9996>}
 JNT_cloth_02_04_L {parent=JNT_cloth_02_03_L, loc=<3.2708 0 0>, rot=<0 0 0 1>}
 JNT_cloth_03_01_L {parent=JNT_Cog, loc=<0.0018 -1.0921 -0.3962>, rot=<0.3018 -0.528 0.455 -0.6505>}
 JNT_cloth_03_02_L {parent=JNT_cloth_03_01_L, loc=<3.4175 0.0017 -0.1892>, rot=<0.0008 0.048 -0.0174 -0.9987>}
 JNT_cloth_03_03_L {parent=JNT_cloth_03_02_L, loc=<3.2707 0 0>, rot=<0 0 0 -1>}
 JNT_cloth_03_04_L {parent=JNT_cloth_03_03_L, loc=<3.271 0 0>, rot=<0 0 0 1>}
 JNT_cloth_04_01_L {parent=JNT_Cog, loc=<-0.1164 -1.1001 0.2277>, rot=<0.4091 -0.4213 0.6289 -0.5096>}
 JNT_cloth_04_02_L {parent=JNT_cloth_04_01_L, loc=<4.213 0 -0.0879>, rot=<0.0003 0.0585 -0.0051 -0.9983>}
 JNT_cloth_04_03_L {parent=JNT_cloth_04_02_L, loc=<3.0208 0 0>, rot=<0 0 0 1>}
 JNT_cloth_04_04_L {parent=JNT_cloth_04_03_L, loc=<2.64 0 0>, rot=<-0.0098 0 0 -1>}
 JNT_cloth_05_01_L {parent=JNT_Cog, loc=<-0.03 -0.7544 0.4543>, rot=<0.491 -0.2782 0.7536 -0.337>}
 JNT_cloth_05_02_L {parent=JNT_cloth_05_01_L, loc=<4.7128 -0.0005 -0.2073>, rot=<-0.0001 0.0574 0.0019 -0.9984>}
 JNT_cloth_05_03_L {parent=JNT_cloth_05_02_L, loc=<3.0208 0 0>, rot=<0 0 0 1>}
 JNT_cloth_05_04_L {parent=JNT_cloth_05_03_L, loc=<3.755 0 0>, rot=<0.0088 0 0 -1>}
 JNT_cloth_06_01_L {parent=JNT_Cog, loc=<-0.132 -0.52 0.7071>, rot=<0.5095 -0.119 0.8415 -0.1348>}
 JNT_cloth_06_02_L {parent=JNT_cloth_06_01_L, loc=<4.2207 0 0.0116>, rot=<-0.0007 0.0894 0.0079 -0.996>}
 JNT_cloth_06_03_L {parent=JNT_cloth_06_02_L, loc=<4.4427 0 0>, rot=<0 0 0.0006 -1>}
 JNT_cloth_06_04_L {parent=JNT_cloth_06_03_L, loc=<4.511 0 0>, rot=<0 0 0 1>}
 JNT_cloth_06_05_L {parent=JNT_cloth_06_04_L, loc=<3.182 0 0>, rot=<-0.0057 0 0 -1>}
 JNT_cloth_06_01_R {parent=JNT_Cog, loc=<-0.1305 0.4506 0.7021>, rot=<0.5101 0.1223 0.8399 0.139>}
 JNT_cloth_06_02_R {parent=JNT_cloth_06_01_R, loc=<4.2208 0 0>, rot=<0.0007 0.088 -0.0079 -0.9961>}
 JNT_cloth_06_03_R {parent=JNT_cloth_06_02_R, loc=<4.4428 0 0>, rot=<0 0 -0.0006 -1>}
 JNT_cloth_06_04_R {parent=JNT_cloth_06_03_R, loc=<4.511 0 0>, rot=<0 0 0 1>}
 JNT_cloth_06_05_R {parent=JNT_cloth_06_04_R, loc=<3.182 0 0>, rot=<0.0057 0 0 -1>}
 JNT_cloth_05_01_R {parent=JNT_Cog, loc=<-0.0277 0.6821 0.447>, rot=<0.4731 0.2887 0.7622 0.3344>}
 JNT_cloth_05_02_R {parent=JNT_cloth_05_01_R, loc=<4.7174 0 0>, rot=<0.0001 0.0793 -0.0019 -0.9969>}
 JNT_cloth_05_03_R {parent=JNT_cloth_05_02_R, loc=<3.0208 0 0>, rot=<0 0 0 1>}
 JNT_cloth_05_04_R {parent=JNT_cloth_05_03_R, loc=<3.755 0 0>, rot=<-0.0088 0 0 -1>}
 JNT_cloth_04_01_R {parent=JNT_Cog, loc=<-0.113 1.0257 0.2168>, rot=<-0.4006 -0.4292 -0.6301 -0.5082>}
 JNT_cloth_04_02_R {parent=JNT_cloth_04_01_R, loc=<4.2139 0 0>, rot=<-0.0003 0.0689 0.0051 -0.9976>}
 JNT_cloth_04_03_R {parent=JNT_cloth_04_02_R, loc=<3.0208 0 0>, rot=<0 0 0 1>}
 JNT_cloth_04_04_R {parent=JNT_cloth_04_03_R, loc=<2.64 0 0>, rot=<0.0098 0 0 -1>}
 JNT_cloth_03_01_R {parent=JNT_Cog, loc=<0.0052 1.011 -0.407>, rot=<-0.2867 -0.5477 -0.4591 -0.638>}
 JNT_cloth_03_02_R {parent=JNT_cloth_03_01_R, loc=<3.4227 0 0>, rot=<-0.0013 0.0756 0.0171 -0.997>}
 JNT_cloth_03_03_R {parent=JNT_cloth_03_02_R, loc=<3.2708 0 0>, rot=<0 0 0 -1>}
 JNT_cloth_03_04_R {parent=JNT_cloth_03_03_R, loc=<3.271 0 0>, rot=<0 0 0 -1>}
 JNT_cloth_02_01_R {parent=JNT_Cog, loc=<0.0359 0.6545 -0.7592>, rot=<-0.1643 -0.6093 -0.254 -0.733>}
 JNT_cloth_02_02_R {parent=JNT_cloth_02_01_R, loc=<3.6955 0 0>, rot=<0.0017 0.0596 -0.028 -0.9978>}
 JNT_cloth_02_03_R {parent=JNT_cloth_02_02_R, loc=<3.2708 0 0>, rot=<0 0 -0.0281 -0.9996>}
 JNT_cloth_02_04_R {parent=JNT_cloth_02_03_R, loc=<4.1094 0 0>, rot=<0 0 0 -1>}
 JNT_cloth_01_01_R {parent=JNT_Cog, loc=<-0.0585 -0.329 -1.4731>, rot=<-0.07 -0.6212 -0.1044 -0.7735>}
 JNT_cloth_01_02_R {parent=JNT_cloth_01_01_R, loc=<3.1166 0 0>, rot=<0.0067 0.0421 -0.1562 -0.9868>}
 JNT_cloth_01_03_R {parent=JNT_cloth_01_02_R, loc=<3.9754 0 0>, rot=<0 0 0.0191 -0.9998>}
 JNT_cloth_01_04_R {parent=JNT_cloth_01_03_R, loc=<4.6628 0 0>, rot=<0.0012 0 0 -1>}
 dyn_cloth_01_01_R_mstr {parent=JNT_Cog, loc=<-0.0585 -0.329 -1.4731>, rot=<-0.4176 -0.4199 -0.5682 -0.5713>}
 dyn_cloth_02_01_R_mstr {parent=JNT_Cog, loc=<0.0359 0.6545 -0.7592>, rot=<-0.4176 -0.4199 -0.5682 -0.5713>}
 dyn_cloth_03_01_R_mstr {parent=JNT_Cog, loc=<0.0052 1.011 -0.407>, rot=<-0.4176 -0.4199 -0.5682 -0.5713>}
 dyn_cloth_04_01_R_mstr {parent=JNT_Cog, loc=<-0.113 1.0257 0.2168>, rot=<-0.4176 -0.4199 -0.5682 -0.5713>}
 dyn_cloth_05_01_R_mstr {parent=JNT_Cog, loc=<-0.0277 0.6821 0.447>, rot=<-0.4176 -0.4199 -0.5682 -0.5713>}
 dyn_cloth_06_01_R_mstr {parent=JNT_Cog, loc=<-0.1305 0.4506 0.7021>, rot=<-0.4176 -0.4199 -0.5682 -0.5713>}
 dyn_cloth_06_01_L_mstr {parent=JNT_Cog, loc=<-0.132 -0.52 0.7071>, rot=<-0.4176 -0.4199 -0.5682 -0.5713>}
 dyn_cloth_05_01_L_mstr {parent=JNT_Cog, loc=<-0.03 -0.7544 0.4543>, rot=<-0.4176 -0.4199 -0.5682 -0.5713>}
 dyn_cloth_04_01_L_mstr {parent=JNT_Cog, loc=<-0.1164 -1.1001 0.2277>, rot=<-0.4176 -0.4199 -0.5682 -0.5713>}
 dyn_cloth_03_01_L_mstr {parent=JNT_Cog, loc=<0.0018 -1.0921 -0.3962>, rot=<-0.4176 -0.4199 -0.5682 -0.5713>}
 dyn_cloth_02_01_L_mstr {parent=JNT_Cog, loc=<0.0337 -0.7394 -0.7521>, rot=<-0.4176 -0.4199 -0.5682 -0.5713>}
 dyn_cloth_01_01_L_mstr {parent=JNT_Cog, loc=<0.0456 -0.2814 -0.9877>, rot=<-0.4176 -0.4199 -0.5682 -0.5713>}
 dyn_cloth_01_B_mstr {parent=JNT_Cog, loc=<-0.0691 -0.0341 0.8467>, rot=<-0.4176 -0.4199 -0.5682 -0.5713>}
 Back_Loc {controlType=EULER_TRANSFORM, parent=root, type=Control, loc=<0 0 0>, rot=<0 0 0 1>}
 Back_Loc_2 {controlType=EULER_TRANSFORM, parent=root, type=Control, loc=<0 0 0>, rot=<0 0 0 1>}
]
variables [SkyDiveCtrl {type=bool, default=False}]
