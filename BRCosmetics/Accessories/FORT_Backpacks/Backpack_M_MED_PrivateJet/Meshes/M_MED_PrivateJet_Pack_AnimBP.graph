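# Animation Blueprint for the PrivateJet backpack cosmetic, adding secondary motion to its cloth, ear and pack bones.
# A cached idle pose runs through chained, rotation-limited Trail controllers, then an offset, spring-driven AnimDynamics on the pack bone and a Control Rig receiving the skydiving flag.
# The event graph sets BwdMovement, the trail alpha, from player movement: it falls as the weighted movement rises, zeroes when jumping or falling, and stays low in the frontend.
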AnimBlueprint /BRCosmetics/Accessories/FORT_Backpacks/Backpack_M_MED_PrivateJet/Meshes/M_MED_PrivateJet_Pack_AnimBP
target_skeleton=/Game/Accessories/FORT_Backpacks/Backpack_M_MED_PrivateJet/Meshes/M_MED_PrivateJet_Pack_Skeleton
graphs [
 AnimGraph {nodes [
   Input Pose()
   n1 = Get BwdMovement() -> (BwdMovement:double)
   n2 = Trail controller - Bone: JNT_cloth_05_B(ComponentPose=Local To Component(LocalPose=Use cached pose 'PostAnim'()), Alpha=@n1) {
    bLimitRotation=true
    chainLength=4
    alphaBoolBlend {blendInTime=10, blendOutTime=10}
    alphaScaleBiasClamp {bInterpResult=true}
    baseJoint {boneName=root}
    relaxationSpeedScaleInputProcessor {bInterpResult=true}
    rotationLimits [
     {limitMax=<0 5 30>, limitMin=<0 -5 -30>}
     {limitMax=<0 5 30>, limitMin=<0 -5 -30>}
     {limitMax=<0 5 60>, limitMin=<0 -5 -60>}
     {limitMax=<0 5 60>, limitMin=<0 -5 -60>}
    ]
    rotationOffsets [<0 0 0>, <0 0 0>, <0 0 0>, <0 0 0>]
    trailBone {boneName=JNT_cloth_05_B}
    trailRelaxationSpeed {editorCurveData {keys [<at 0 30 RCIM_Linear>,<at 1 15 RCIM_Linear>]}}
   }
   n3 = Trail controller - Bone: JNT_cloth_01_04_L(ComponentPose=@n2, Alpha=@n1) {
    bLimitRotation=true
    chainLength=3
    alphaBoolBlend {blendInTime=10, blendOutTime=10}
    alphaScaleBiasClamp {bInterpResult=true}
    baseJoint {boneName=root}
    relaxationSpeedScaleInputProcessor {bInterpResult=true}
    rotationLimits [{limitMax=<0 60 60>, limitMin=<0 -60 -60>}, {limitMax=<0 60 60>, limitMin=<0 -60 -60>}, {limitMax=<180 180 180>}]
    rotationOffsets [<0 0 0>, <0 0 0>, <0 0 0>]
    trailBone {boneName=JNT_cloth_01_04_L}
    trailRelaxationSpeed {editorCurveData {keys [<at 0 30 RCIM_Linear>,<at 1 15 RCIM_Linear>]}}
   }
   n4 = Trail controller - Bone: JNT_cloth_02_04_L(ComponentPose=@n3, Alpha=@n1) {
    bLimitRotation=true
    chainLength=3
    alphaBoolBlend {blendInTime=10, blendOutTime=10}
    alphaScaleBiasClamp {bInterpResult=true}
    baseJoint {boneName=root}
    relaxationSpeedScaleInputProcessor {bInterpResult=true}
    rotationLimits [{limitMax=<0 60 60>, limitMin=<0 -60 -60>}, {limitMax=<0 60 60>, limitMin=<0 -60 -60>}, {limitMax=<180 180 180>}]
    rotationOffsets [<0 0 0>, <0 0 0>, <0 0 0>]
    trailBone {boneName=JNT_cloth_02_04_L}
    trailRelaxationSpeed {editorCurveData {keys [<at 0 30 RCIM_Linear>,<at 1 15 RCIM_Linear>]}}
   }
   n5 = Trail controller - Bone: JNT_cloth_03_04_L(ComponentPose=@n4, Alpha=@n1) {
    bLimitRotation=true
    chainLength=3
    alphaBoolBlend {blendInTime=10, blendOutTime=10}
    alphaScaleBiasClamp {bInterpResult=true}
    baseJoint {boneName=root}
    relaxationSpeedScaleInputProcessor {bInterpResult=true}
    rotationLimits [{limitMax=<0 60 60>, limitMin=<0 -60 -60>}, {limitMax=<0 60 60>, limitMin=<0 -60 -60>}, {limitMax=<180 180 180>}]
    rotationOffsets [<0 0 0>, <0 0 0>, <0 0 0>]
    trailBone {boneName=JNT_cloth_03_04_L}
    trailRelaxationSpeed {editorCurveData {keys [<at 0 30 RCIM_Linear>,<at 1 15 RCIM_Linear>]}}
   }
   n6 = Trail controller - Bone: JNT_cloth_04_04_L(ComponentPose=@n5, Alpha=@n1) {
    bLimitRotation=true
    chainLength=3
    alphaBoolBlend {blendInTime=10, blendOutTime=10}
    alphaScaleBiasClamp {bInterpResult=true}
    baseJoint {boneName=root}
    relaxationSpeedScaleInputProcessor {bInterpResult=true}
    rotationLimits [{limitMax=<0 60 60>, limitMin=<0 -60 -60>}, {limitMax=<0 60 60>, limitMin=<0 -60 -60>}, {limitMax=<180 180 180>}]
    rotationOffsets [<0 0 0>, <0 0 0>, <0 0 0>]
    trailBone {boneName=JNT_cloth_04_04_L}
    trailRelaxationSpeed {editorCurveData {keys [<at 0 30 RCIM_Linear>,<at 1 15 RCIM_Linear>]}}
   }
   n7 = Trail controller - Bone: JNT_cloth_05_04_L(ComponentPose=@n6, Alpha=@n1) {
    bLimitRotation=true
    chainLength=3
    alphaBoolBlend {blendInTime=10, blendOutTime=10}
    alphaScaleBiasClamp {bInterpResult=true}
    baseJoint {boneName=root}
    relaxationSpeedScaleInputProcessor {bInterpResult=true}
    rotationLimits [{limitMax=<0 60 60>, limitMin=<0 -60 -60>}, {limitMax=<0 60 60>, limitMin=<0 -60 -60>}, {limitMax=<180 180 180>}]
    rotationOffsets [<0 0 0>, <0 0 0>, <0 0 0>]
    trailBone {boneName=JNT_cloth_05_04_L}
    trailRelaxationSpeed {editorCurveData {keys [<at 0 30 RCIM_Linear>,<at 1 15 RCIM_Linear>]}}
   }
   n8 = Trail controller - Bone: JNT_cloth_06_05_L(ComponentPose=@n7, Alpha=@n1) {
    bLimitRotation=true
    chainLength=4
    alphaBoolBlend {blendInTime=10, blendOutTime=10}
    alphaScaleBiasClamp {bInterpResult=true}
    baseJoint {boneName=root}
    relaxationSpeedScaleInputProcessor {bInterpResult=true}
    rotationLimits [
     {limitMax=<0 5 30>, limitMin=<0 -5 -30>}
     {limitMax=<0 5 30>, limitMin=<0 -5 -30>}
     {limitMax=<0 5 30>, limitMin=<0 -5 -30>}
     {limitMax=<0 5 30>, limitMin=<0 -5 -30>}
    ]
    rotationOffsets [<0 0 0>, <0 0 0>, <0 0 0>, <0 0 0>]
    trailBone {boneName=JNT_cloth_06_05_L}
    trailRelaxationSpeed {editorCurveData {keys [<at 0 30 RCIM_Linear>,<at 1 15 RCIM_Linear>]}}
   }
   n9 = Trail controller - Bone: JNT_cloth_06_05_L(ComponentPose=@n8, Alpha=@n1) {
    bLimitRotation=true
    chainLength=4
    alphaBoolBlend {blendInTime=10, blendOutTime=10}
    alphaScaleBiasClamp {bInterpResult=true}
    baseJoint {boneName=root}
    relaxationSpeedScaleInputProcessor {bInterpResult=true}
    rotationLimits [
     {limitMax=<0 5 30>, limitMin=<0 -5 -30>}
     {limitMax=<0 5 30>, limitMin=<0 -5 -30>}
     {limitMax=<0 5 30>, limitMin=<0 -5 -30>}
     {limitMax=<0 5 30>, limitMin=<0 -5 -30>}
    ]
    rotationOffsets [<0 0 0>, <0 0 0>, <0 0 0>, <0 0 0>]
    trailBone {boneName=JNT_cloth_06_05_L}
    trailRelaxationSpeed {editorCurveData {keys [<at 0 30 RCIM_Linear>,<at 1 15 RCIM_Linear>]}}
   }
   n10 = Trail controller - Bone: JNT_cloth_05_04_R(ComponentPose=@n9, Alpha=@n1) {
    bLimitRotation=true
    chainLength=3
    alphaBoolBlend {blendInTime=10, blendOutTime=10}
    alphaScaleBiasClamp {bInterpResult=true}
    baseJoint {boneName=root}
    relaxationSpeedScaleInputProcessor {bInterpResult=true}
    rotationLimits [{limitMax=<0 60 60>, limitMin=<0 -60 -60>}, {limitMax=<0 60 60>, limitMin=<0 -60 -60>}, {limitMax=<180 180 180>}]
    rotationOffsets [<0 0 0>, <0 0 0>, <0 0 0>]
    trailBone {boneName=JNT_cloth_05_04_R}
    trailRelaxationSpeed {editorCurveData {keys [<at 0 30 RCIM_Linear>,<at 1 15 RCIM_Linear>]}}
   }
   n11 = Trail controller - Bone: JNT_cloth_04_04_R(ComponentPose=@n10, Alpha=@n1) {
    bLimitRotation=true
    chainLength=3
    alphaBoolBlend {blendInTime=10, blendOutTime=10}
    alphaScaleBiasClamp {bInterpResult=true}
    baseJoint {boneName=root}
    relaxationSpeedScaleInputProcessor {bInterpResult=true}
    rotationLimits [{limitMax=<0 60 60>, limitMin=<0 -60 -60>}, {limitMax=<0 60 60>, limitMin=<0 -60 -60>}, {limitMax=<180 180 180>}]
    rotationOffsets [<0 0 0>, <0 0 0>, <0 0 0>]
    trailBone {boneName=JNT_cloth_04_04_R}
    trailRelaxationSpeed {editorCurveData {keys [<at 0 30 RCIM_Linear>,<at 1 15 RCIM_Linear>]}}
   }
   n12 = Trail controller - Bone: JNT_cloth_03_04_R(ComponentPose=@n11, Alpha=@n1) {
    bLimitRotation=true
    chainLength=3
    alphaBoolBlend {blendInTime=10, blendOutTime=10}
    alphaScaleBiasClamp {bInterpResult=true}
    baseJoint {boneName=root}
    relaxationSpeedScaleInputProcessor {bInterpResult=true}
    rotationLimits [{limitMax=<0 60 60>, limitMin=<0 -60 -60>}, {limitMax=<0 60 60>, limitMin=<0 -60 -60>}, {limitMax=<180 180 180>}]
    rotationOffsets [<0 0 0>, <0 0 0>, <0 0 0>]
    trailBone {boneName=JNT_cloth_03_04_R}
    trailRelaxationSpeed {editorCurveData {keys [<at 0 30 RCIM_Linear>,<at 1 15 RCIM_Linear>]}}
   }
   n13 = Trail controller - Bone: JNT_cloth_02_04_R(ComponentPose=@n12, Alpha=@n1) {
    bLimitRotation=true
    chainLength=3
    alphaBoolBlend {blendInTime=10, blendOutTime=10}
    alphaScaleBiasClamp {bInterpResult=true}
    baseJoint {boneName=root}
    relaxationSpeedScaleInputProcessor {bInterpResult=true}
    rotationLimits [{limitMax=<0 60 60>, limitMin=<0 -60 -60>}, {limitMax=<0 60 60>, limitMin=<0 -60 -60>}, {limitMax=<180 180 180>}]
    rotationOffsets [<0 0 0>, <0 0 0>, <0 0 0>]
    trailBone {boneName=JNT_cloth_02_04_R}
    trailRelaxationSpeed {editorCurveData {keys [<at 0 30 RCIM_Linear>,<at 1 15 RCIM_Linear>]}}
   }
   n14 = Trail controller - Bone: JNT_cloth_01_04_R(ComponentPose=@n13, Alpha=@n1) {
    bLimitRotation=true
    chainLength=3
    alphaBoolBlend {blendInTime=10, blendOutTime=10}
    alphaScaleBiasClamp {bInterpResult=true}
    baseJoint {boneName=root}
    relaxationSpeedScaleInputProcessor {bInterpResult=true}
    rotationLimits [{limitMax=<0 60 60>, limitMin=<0 -60 -60>}, {limitMax=<0 60 60>, limitMin=<0 -60 -60>}, {limitMax=<180 180 180>}]
    rotationOffsets [<0 0 0>, <0 0 0>, <0 0 0>]
    trailBone {boneName=JNT_cloth_01_04_R}
    trailRelaxationSpeed {editorCurveData {keys [<at 0 30 RCIM_Linear>,<at 1 15 RCIM_Linear>]}}
   }
   n15 = Get BwdMovement()
   n16 = Trail controller - Bone: JNT_ear_08_L(ComponentPose=@n14, Alpha=@n15) {
    bLimitRotation=true
    chainLength=7
    rotationLimits [
     {limitMax=<5 45 45>, limitMin=<-5 -45 -45>}
     {limitMax=<5 45 45>, limitMin=<-5 -45 -45>}
     {limitMax=<5 45 45>, limitMin=<-5 -45 -45>}
     {limitMax=<5 45 45>, limitMin=<-5 -45 -45>}
     {limitMax=<5 45 45>, limitMin=<-5 -45 -45>}
     {limitMax=<5 45 45>, limitMin=<-5 -45 -45>}
     {limitMax=<5 45 45>, limitMin=<-5 -45 -45>}
    ]
    rotationOffsets [<0 0 0>, <0 0 0>, <0 0 0>, <0 0 0>, <0 0 0>, <0 0 0>, <0 0 0>]
    trailBone {boneName=JNT_ear_08_L}
    trailRelaxationSpeed {editorCurveData {keys [<at 0 30 RCIM_Linear>,<at 1 15 RCIM_Linear>]}}
   }
   n17 = Trail controller - Bone: JNT_ear_08_R(ComponentPose=@n16, Alpha=@n15) {
    bLimitRotation=true
    chainLength=7
    rotationLimits [
     {limitMax=<5 45 45>, limitMin=<-5 -45 -45>}
     {limitMax=<5 45 45>, limitMin=<-5 -45 -45>}
     {limitMax=<5 45 45>, limitMin=<-5 -45 -45>}
     {limitMax=<5 45 45>, limitMin=<-5 -45 -45>}
     {limitMax=<5 45 45>, limitMin=<-5 -45 -45>}
     {limitMax=<5 45 45>, limitMin=<-5 -45 -45>}
     {limitMax=<5 45 45>, limitMin=<-5 -45 -45>}
    ]
    rotationOffsets [<0 0 0>, <0 0 0>, <0 0 0>, <0 0 0>, <0 0 0>, <0 0 0>, <0 0 0>]
    trailBone {boneName=JNT_ear_08_R}
    trailRelaxationSpeed {editorCurveData {keys [<at 0 30 RCIM_Linear>,<at 1 15 RCIM_Linear>]}}
   }
   n18 = "Transform (Modify) Bone - Bone: pack"(ComponentPose=@n17, Translation="-0.000,0.000,5.242", Rotation="0,0,0") {translationMode=BMM_Additive, alphaBoolBlend {blendOption=Linear}, boneToModify {boneName=pack}, translation {x=0, z=5.2422}}
   n19 = Anim Dynamics - Bone: pack(ComponentPose=@n18, Alpha=0.4, ExternalForce="0,0,0") {
    angularDampingOverride=1
    angularSpringConstant=200
    bAngularSpring=true
    bGravityOverrideInSimSpace=true
    bLinearSpring=true
    bOverrideAngularDamping=true
    bOverrideLinearDamping=true
    bUseGravityOverride=true
    bUsePlanarLimit=false
    gravityScale=0.35
    linearDampingOverride=1
    linearSpringConstant=100
    simulationSpace=RootRelative
    alphaBoolBlend {blendOption=Linear}
    boundBone {boneName=pack}
    componentAppliedLinearAccClamp=<30000 30000 30000>
    componentLinearAccScale=<0.25 0.25 0.35>
    componentLinearVelScale=<0.25 0.25 0.35>
    gravityOverride {z=-980}
    physicsBodyDefinitions [{
      sphereCollisionRadius=0
      boundBone {boneName=pack}
      boxExtents=<5 10 10>
      constraintSetup {angularLimitsMax=<0 15 15>, angularLimitsMin=<0 -15 -15>, angularTarget=<1 0 0>, linearAxesMax=<30 0 10>, linearAxesMin=<-30 -20 -20>}
      localJointOffset=<0 0 -1>
     }]
   }
   n20 = Control Rig!(Source=Component To Local(ComponentPose=@n19), Alpha, SkyDiveCtrl:bool=Get bIsSkyDiving()) -> (Pose) {
    alphaBoolBlend {blendOption=Linear}
    controlRigAssetReference {blueprintRigClass=&/M_MED_PrivateJet_Pack_CtrlRig.M_MED_PrivateJet_Pack_CtrlRig_C}
   }
   Output Pose(Result=@n20) {layerGroup=None}
   Save cached pose 'PostAnim'(Pose=Sequence Player 'PrivateJet_Idle_Anim_PACK'() {sequence=/BRCosmetics/Accessories/FORT_Backpacks/Backpack_M_MED_PrivateJet/Animation/PrivateJet_Idle_Anim_PACK}) {cacheName=PostAnim}
   Sequence Player 'PrivateJet_BasePose_Anim_PACK'() {bLoopAnimation=false, sequence=/Game/Accessories/FORT_Backpacks/Backpack_M_MED_PrivateJet/Animation/PrivateJet_BasePose_Anim_PACK}
   Get DynamicAlpha() -> (DynamicAlpha:double)
  ]}
 EventGraph {nodes [
   Event BlueprintUpdateAnimation()
   > n1 = Branch(Condition=Get bIsInFrontEnd())
   > Set BwdMovement(BwdMovement:double=0.1) -> (Output_Get:double)
   n2 = PlayerMovement(execute=@n1.else, AnimInstance=Self-Reference(), XDotProductCoefficient=1, YDotProductCoefficient=1, ZDotProductCoefficient=1)
   n3 = float + float(A=float * float(A="Absolute (Float)"(A=@n2), B=0.5), B=MapRangeClamped(Value=@n2.YMovement, InRangeA=0, InRangeB=-1)) {is=title K2Node_CommutativeAssociativeBinaryOperator}
   n4 = MapRangeClamped(Value=@n3, InRangeA=0, InRangeB=1.1, OutRangeA=0.7, OutRangeB=0)
   n5 = Get bIsJumping()
   n6 = Get bIsFalling()
   > Set BwdMovement(BwdMovement=SelectFloat(A=0, B=@n4, bPickA=OR Boolean(A=@n6, B=@n5)))
   > Set FwdMovement(FwdMovement:double=MapRangeClamped(Value=@n3, InRangeA=0, InRangeB=1.1)) -> (Output_Get:double)
   > Branch(Condition=Get bIsCrouching())
   Sequence()
   > Set DynamicAlpha(DynamicAlpha:double=0) -> (Output_Get:double)
   TryGetPawnOwner()
   "Clamp (Float)"(Value=float - float(A=@n4, B="Clamp (Float)"(Value=@n2.YMovement, Min=0, Max=1)), Min=0, Max=1)
  ]}
]
variables [
 BwdMovement {type=number}
 DynamicAlpha {type=number, default=1}
 FwdMovement {type=number}
]
